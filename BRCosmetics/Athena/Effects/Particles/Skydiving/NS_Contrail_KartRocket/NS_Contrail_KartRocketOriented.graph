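# Skydiving contrail cosmetic Niagara system for the Kart Rocket: tire effects, long mesh trails and red-tinted dive dust.
# Tire and trail particles follow skeletal mesh bones with offsets per spawn group, easing in by emitter age.
# System dive value and timelines drive their tilt, sideways slide and position, while the glide lean parameter stretches the long trails.
# Dust emitters read tire emitter particles to spawn sprites, gated by a dive threshold.
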NiagaraSystem /BRCosmetics/Athena/Effects/Particles/Skydiving/NS_Contrail_KartRocket/NS_Contrail_KartRocketOriented
dependencies {
 usedDataInterfaces [
  &NiagaraDataInterfaceArrayFloat
  &NiagaraDataInterfaceArrayInt32
  &NiagaraDataInterfaceArrayFloat3
  &NiagaraDataInterfaceCurve
  &NiagaraDataInterfaceMeshRendererInfo
  &NiagaraDataInterfaceVectorCurve
  &NiagaraDataInterfaceParticleRead
  &NiagaraDataInterfaceColorCurve
  &NiagaraDataInterfaceSpriteRendererInfo
  &NiagaraDataInterfaceSkeletalMesh
 ]
 usedDynamicInputs [
  &SelectFloatFromArray
  &Multiply_VectorByFloat
  &FN_FrontEndFloatModifier
  &DotProduct
  &NormalizeVector
  &Abs_Float
  &OneMinusFloat
  &ClampFloat
  /Game/Effects/Niagara/DynamicInputs/Engine/FN_GetSystemAxis
  &Waveform
 ]
 usedModules [
  &self:SystemSpawnScript.SystemScriptSource.SystemScriptGraph.NiagaraNodeAssignment_5.SetVariables_#1
  &self:SystemSpawnScript.SystemScriptSource.SystemScriptGraph.NiagaraNodeAssignment_1.SetVariables_#2
  /Game/Effects/Niagara/Modules/Skydiving/FN_DivingAttributes/v5/FN_DivingAttributes
  &self:UpdatePreviousRotation
  &self:SystemSpawnScript.SystemScriptSource.SystemScriptGraph.NiagaraNodeAssignment_0.SetVariables_#3
  &FN_Integrator
  &SystemState
  &Timeline
  &InterpolateOverTime
  &self:SystemSpawnScript.SystemScriptSource.SystemScriptGraph.NiagaraNodeAssignment_2.SetVariables_#4
  &self:ScratchModule_01
  &self:SystemSpawnScript.SystemScriptSource.SystemScriptGraph.NiagaraNodeAssignment_3.SetVariables_#5
  &self:SystemSpawnScript.SystemScriptSource.SystemScriptGraph.NiagaraNodeAssignment_6.SetVariables_#6
  &self:SystemSpawnScript.SystemScriptSource.SystemScriptGraph.NiagaraNodeAssignment_7.SetVariables_#7
 ]
}
emitters {
 Dive_Dust_Back {
  emitterUpdateScript [
   &EmitterState()
   &SpawnParticlesFromOtherEmitter(
    Attribute Reader=DataInterface() {propertyValues {EmitterBinding {bindingMode=Other, emitterName=Tire_Back_02}}}
    Spawn Rate=&Multiply_Float(
     A=120
     B=&MakeCustomFloatFromBool(Bool=&SetBoolByFloatComparison(A=$System.Dive, B=0.7), True Float=1, False Float=0)
    )
    Use Maximum Age=false
    Use Minimum Age=false
   )
  ]
  particleSpawnScript [
   &InitializeParticle(
    Color=<rgba 1 1 1 0.4>
    Lifetime Max=0.35
    Lifetime Min=0.2
    Lifetime Mode=::NewEnumerator1
    Sprite Rotation Mode=::NewEnumerator2
    Sprite Size Mode=::NewEnumerator1
    Sprite UV Mode=::NewEnumerator1
    Uniform Sprite Size Max=150
    Uniform Sprite Size Min=110
    UsePositionOffset=false
    Write Material Random=true
   )
   &SampleParticlesFromOtherEmitter()
   &AddVelocity(Velocity=&RandomRangeVector(Minimum=<0 -100 450>, Maximum=<0 -100 500>)) {module=AddVelocity001}
   &CameraOffset(Camera Offset Amount=-5)
  ]
  particleUpdateScript [
   &ParticleState()
   &Drag(Use Linear Drag=true, Use Rotational Drag=true)
   &InheritSourceMovement(
    Applied Position Scale=<-0.1 -0.1 -0.1>
    ApplyForce=false
    ApplyPosition=true
    ApplyPositionFalloff=false
    Source Velocity:Vector3f=&TransformVector(Vector=$Engine.Owner.Velocity, Source Space=::World, Destination Space=::Local)
   )
   &SolveForcesAndVelocity(Clamp Velocity=false, Limit Acceleration=false)
   &ScaleColor(
    Scale Alpha=&Multiply_Float(
     A=&FloatFromCurve(FloatCurve=DataInterface() {propertyValues {Curve {keys [
          <at 0 0 RCIM_Cubic>
          <at 0.025 1 RCIM_Cubic>
          {arriveTangent=-3.6291, interpMode=RCIM_Cubic, leaveTangent=-3.6291, tangentMode=RCTM_User, time=1, value=0}
         ]}}})
     B=0.2
    )
    Scale RGB=&MakeVectorFromLinearColorRGB(Color=&ColorFromCurve(ColorCurve=DataInterface() {propertyValues {
        AlphaCurve {keys [<at 0 1 RCIM_Cubic>]}
        BlueCurve {keys [
          <at 0.052 0 RCIM_Cubic>
          {arriveTangent=0.1692, leaveTangent=0.026, tangentMode=RCTM_User, time=0.3845, value=0}
          <at 0.6543 0>
          <at 0.9837 0>
         ]}
        GreenCurve {keys [
          <at 0.052 0 RCIM_Cubic>
          {arriveTangent=-0.049, leaveTangent=-0.1372, tangentMode=RCTM_User, time=0.3845, value=0}
          <at 0.6543 0>
          <at 0.9837 0>
         ]}
        RedCurve {keys [<at 0.052 0.6719 RCIM_Cubic>, {arriveTangent=-1.2191, time=0.3845, value=0.0885}, <at 0.6543 0.0573>, <at 0.9837 0.0156>]}
       }}))
    ScaleA=true
    ScaleRGB=true
   )
   &ScaleSpriteSize(Uniform Curve Sprite Scale=DataInterface() {propertyValues {Curve {keys [{arriveTangent=2.3673, interpMode=RCIM_Cubic, time=0, value=0.25}, {interpMode=RCIM_Cubic, leaveTangent=5.0333, time=1, value=1.25}]}}})
   &SpriteRotationRate(Rotation Rate=&RandomRangeFloat(Minimum=-60, Maximum=60))
   &SubUVAnimation(
    Play Rate:NiagaraFloat=&RandomRangeFloat(Minimum=-0.25, Maximum=-0.5)
    Playback Scale Curve=DataInterface() {propertyValues {Curve {keys [
        {interpMode=RCIM_Cubic, tangentMode=RCTM_User, time=0.25, value=1}
        {arriveTangent=-1.2656, interpMode=RCIM_Cubic, leaveTangent=-1.2656, tangentMode=RCTM_User, time=1, value=0}
       ]}}}
    Random Start Frame=true
    "Sprite Renderer "=DataInterface() {propertyValues {SpriteRenderer=&self:Dive_Dust_Back_0.NiagaraSpriteRendererProperties_1}}
    Start Frame Offset:NiagaraInt32=&RandomRangeInt(Minimum=0, Maximum=63)
    SubUV Animation Mode:ENiagara_SubUVLookupModeV2=::NewEnumerator3
    UseEndFrame=true
    UsePlayRateScale=false
    UseStartFrame=true
   )
   &DynamicMaterialParameters(
    Index 0 Param 1=&FloatFromCurve(FloatCurve=DataInterface() {propertyValues {Curve {keys [
         {interpMode=RCIM_Cubic, tangentMode=RCTM_User, time=0.2, value=0.2}
         {arriveTangent=0.4832, interpMode=RCIM_Cubic, leaveTangent=0.4832, tangentMode=RCTM_User, time=1, value=1}
        ]}}})
    Index 0 Param 2=1
    Index 0 Param 3=&RandomRangeFloat(Minimum=0.25, Maximum=0.5)
    Index 1 Param 1=20
    Index 1 Param 2=250
    Module.Index0Param0Write=true
    Module.Index0Param1Write=true
    Module.Index0Param2Write=true
    Module.Index0Param3Write=true
    Module.Index1Param0Write=true
    Module.Index1Param1Write=true
    Module.Index1Param2Write=false
    Module.Index1Param3Write=false
    Write Parameter Index 1=true
   )
  ]
  rendererClasses [&NiagaraSpriteRendererProperties]
  renderers [{rendererClass=&NiagaraSpriteRendererProperties}]
 }
 Dive_Dust_Front {
  emitterUpdateScript [
   &EmitterState()
   &SpawnParticlesFromOtherEmitter(
    Attribute Reader=DataInterface() {propertyValues {EmitterBinding {bindingMode=Other, emitterName=Tire_Front}}}
    Spawn Rate=&Multiply_Float(
     A=12
     B=&MakeCustomFloatFromBool(Bool=&SetBoolByFloatComparison(A=$System.Dive, B=0.7), True Float=1, False Float=0)
    )
    Use Maximum Age=false
    Use Minimum Age=false
   )
  ]
  particleSpawnScript [
   &InitializeParticle(
    Lifetime Max=0.35
    Lifetime Min=0.2
    Lifetime Mode=::NewEnumerator1
    Sprite Rotation Mode=::NewEnumerator2
    Sprite Size Mode=::NewEnumerator1
    Sprite UV Mode=::NewEnumerator1
    Uniform Sprite Size Max=150
    Uniform Sprite Size Min=110
    UsePositionOffset=true
    Write Material Random=true
   )
   &SampleParticlesFromOtherEmitter()
   &PartitionParticles()
   &AddVelocity(Velocity=&RandomRangeVector(Minimum=<0 -100 450>, Maximum=<0 -100 500>)) {module=AddVelocity001}
   &CameraOffset(Camera Offset Amount=-5)
  ]
  particleUpdateScript [
   &ParticleState()
   &Drag(Use Linear Drag=true, Use Rotational Drag=true)
   &InheritSourceMovement(
    Applied Position Scale=<-0.1 -0.1 -0.1>
    ApplyForce=false
    ApplyPosition=true
    ApplyPositionFalloff=false
    Source Velocity:Vector3f=&TransformVector(Vector=$Engine.Owner.Velocity, Source Space=::World, Destination Space=::Local)
   )
   &SolveForcesAndVelocity(Clamp Velocity=false, Limit Acceleration=false)
   &ScaleColor(
    Scale Alpha=&Multiply_Float(
     A=&FloatFromCurve(FloatCurve=DataInterface() {propertyValues {Curve {keys [
          <at 0 0 RCIM_Cubic>
          <at 0.1 1 RCIM_Cubic>
          {arriveTangent=-3.6291, interpMode=RCIM_Cubic, leaveTangent=-3.6291, tangentMode=RCTM_User, time=1, value=0}
         ]}}})
     B=0.2
    )
    Scale RGB=&MakeVectorFromLinearColorRGB(Color=&ColorFromCurve(ColorCurve=DataInterface() {propertyValues {
        AlphaCurve {keys [<at 0 1 RCIM_Cubic>]}
        BlueCurve {keys [
          <at 0.052 0 RCIM_Cubic>
          {arriveTangent=0.1692, leaveTangent=0.026, tangentMode=RCTM_User, time=0.3845, value=0}
          <at 0.6543 0>
          <at 0.9837 0>
         ]}
        GreenCurve {keys [
          <at 0.052 0 RCIM_Cubic>
          {arriveTangent=-0.049, leaveTangent=-0.1372, tangentMode=RCTM_User, time=0.3845, value=0}
          <at 0.6543 0>
          <at 0.9837 0>
         ]}
        RedCurve {keys [<at 0.052 0.6719 RCIM_Cubic>, {arriveTangent=-1.2191, time=0.3845, value=0.0885}, <at 0.6543 0.0573>, <at 0.9837 0.0156>]}
       }}))
    ScaleA=true
    ScaleRGB=true
   )
   &ScaleSpriteSize(Uniform Curve Sprite Scale=DataInterface() {propertyValues {Curve {keys [{arriveTangent=2.3673, interpMode=RCIM_Cubic, time=0, value=0.25}, {interpMode=RCIM_Cubic, leaveTangent=5.0333, time=1, value=1.25}]}}})
   &SpriteRotationRate(Rotation Rate=&RandomRangeFloat(Minimum=-60, Maximum=60))
   &SubUVAnimation(
    Play Rate:NiagaraFloat=&RandomRangeFloat(Minimum=-0.25, Maximum=-0.5)
    Playback Scale Curve=DataInterface() {propertyValues {Curve {keys [
        {interpMode=RCIM_Cubic, tangentMode=RCTM_User, time=0.25, value=1}
        {arriveTangent=-1.2656, interpMode=RCIM_Cubic, leaveTangent=-1.2656, tangentMode=RCTM_User, time=1, value=0}
       ]}}}
    Random Start Frame=true
    "Sprite Renderer "=DataInterface() {propertyValues {SpriteRenderer=&self:Dive_Dust_Front_0.NiagaraSpriteRendererProperties_1}}
    Start Frame Offset:NiagaraInt32=&RandomRangeInt(Minimum=0, Maximum=63)
    SubUV Animation Mode:ENiagara_SubUVLookupModeV2=::NewEnumerator3
    UseEndFrame=true
    UsePlayRateScale=false
    UseStartFrame=true
   )
   &DynamicMaterialParameters(
    Index 0 Param 1=&FloatFromCurve(FloatCurve=DataInterface() {propertyValues {Curve {keys [
         {interpMode=RCIM_Cubic, tangentMode=RCTM_User, time=0.2, value=0.2}
         {arriveTangent=0.4832, interpMode=RCIM_Cubic, leaveTangent=0.4832, tangentMode=RCTM_User, time=1, value=1}
        ]}}})
    Index 0 Param 2=1
    Index 0 Param 3=&RandomRangeFloat(Minimum=0.25, Maximum=0.5)
    Index 1 Param 1=20
    Index 1 Param 2=250
    Module.Index0Param0Write=true
    Module.Index0Param1Write=true
    Module.Index0Param2Write=true
    Module.Index0Param3Write=true
    Module.Index1Param0Write=true
    Module.Index1Param1Write=true
    Module.Index1Param2Write=false
    Module.Index1Param3Write=false
    Write Parameter Index 1=true
   )
  ]
  rendererClasses [&NiagaraSpriteRendererProperties]
  renderers [{rendererClass=&NiagaraSpriteRendererProperties}]
 }
 Long_Trail {
  emitterUpdateScript [
   &EmitterState(
    Inactive Response=::NewEnumerator1
    Life Cycle Mode=::NewEnumerator1
    Loop Behavior=::NewEnumerator1
    Loop Duration Mode=::NewEnumerator1
    UseLoopDelay=false
   )
   &SpawnBurst_Instantaneous(Spawn Count=2, Spawn Time=0.9, Use Spawn Probability=false, UseLoopCountLimit=false)
   &SpawnBurst_Instantaneous(Spawn Count=2, Spawn Group=1, Spawn Time=0.9, Use Spawn Probability=false, UseLoopCountLimit=false) {module=SpawnBurst_Instantaneous001}
   &SpawnBurst_Instantaneous(Spawn Count=2, Spawn Group=2, Spawn Time=0.9, Use Spawn Probability=false, UseLoopCountLimit=false) {module=SpawnBurst_Instantaneous002}
   &SpawnBurst_Instantaneous(Spawn Count=2, Spawn Group=3, Spawn Time=0.9, Use Spawn Probability=false, UseLoopCountLimit=false) {module=SpawnBurst_Instantaneous003}
  ]
  particleSpawnScript [
   /SetVariables(Particles.VisibilityTag=&SpawnGroup()) {isSetParameters=true, module=SetVariables_#8}
   &InitializeParticle(
    Color=&MaskLinearColorBySpawnGroup(
     Masked Linear Color=<rgba 30 5.715 0 0.4>
     Passthrough Linear Color=&MaskLinearColor(
      Mask Source=0
      Mask Parameter=?
      Mask=1
      Masked Linear Color=<rgba 0 0 0 3.5>
      Passthrough Linear Color=&MaskLinearColorBySpawnGroup(Masked Linear Color=<rgba 30 5.715 0 0.4>, Passthrough Linear Color=<rgba 0 0 0 3>, Spawn Group Mask=2)
     )
    )
    Lifetime=5
    Mesh Scale=&MaskVectorBySpawnGroup(
     Masked Vector=<0.11 0.875 0.66>
     Passthrough Vector=&MaskVectorBySpawnGroup(
      Masked Vector=<0.11 0.875 0.8>
      Passthrough Vector=&MaskVectorBySpawnGroup(Masked Vector=<0.11 0.875 0.66>, Passthrough Vector=<0.11 0.875 0.8>)
      Spawn Group Mask=1
     )
    )
    Mesh Scale Mode=::NewEnumerator4
    Ribbon Width=0
    Ribbon Width Mode=::NewEnumerator1
    UsePositionOffset=false
    Write Material Random=true
   ) {module=InitializeParticle001}
   &InitialMeshOrientation(
    Mesh Orientation Mode=::NewEnumerator2
    Rotation=&MaskVectorBySpawnGroup(
     Masked Vector=<0.032 0.02 0>
     Passthrough Vector=&MaskVectorBySpawnGroup(
      Masked Vector=<0.032 0.02 0>
      Passthrough Vector=&MaskVectorBySpawnGroup(Masked Vector=<0.032 -0.02 0>, Passthrough Vector=<0.032 -0.02 0>, Spawn Group Mask=2)
      Spawn Group Mask=1
     )
    )
    Use Rotation Vector=true
   ) {module=InitialMeshOrientation001}
   &PartitionParticles()
  ]
  particleUpdateScript [
   &ParticleState(Kill Particles When Lifetime Has Elapsed=false, Let Infinitely Lived Particles Die When Emitter Deactivates=true)
   &SolveForcesAndVelocity(Clamp Velocity=false, Limit Acceleration=false)
   &SkeletalMeshLocation(
    Bone / Socket Index=&SelectIntFromArray(Direct Array Index=$Particles.PartitionParticles.Partition, Int Selection Array=DataInterface() {propertyValues {IntData []}})
    Bone Sampling Mode=::NewEnumerator4
    MaskSpawnGroup=false
    Orientation Sampling=::NewEnumerator1
    Sampled Position Offset=&MaskVectorBySpawnGroup(
     Masked Vector=&SelectVectorFromArray(
      Direct Array Index=$Particles.PartitionParticles.Partition
      Vector Selection Array=DataInterface() {propertyValues {FloatData [<-37.2375 52 -60>, <-37.2375 -130.9564 -58>]}}
     )
     Passthrough Vector=&MaskVectorBySpawnGroup(
      Masked Vector=&SelectVectorFromArray(
       Direct Array Index=$Particles.PartitionParticles.Partition
       Vector Selection Array=DataInterface() {propertyValues {FloatData [<-40 52 -70>, <-38 -127.2045 -70>]}}
      )
      Passthrough Vector=&MaskVectorBySpawnGroup(
       Masked Vector=&SelectVectorFromArray(
        Direct Array Index=$Particles.PartitionParticles.Partition
        Vector Selection Array=DataInterface() {propertyValues {FloatData [<-42.8999 50 -60>, <-42.0126 -132.0263 -58>]}}
       )
       Passthrough Vector=&SelectVectorFromArray(
        Direct Array Index=$Particles.PartitionParticles.Partition
        Vector Selection Array=DataInterface() {propertyValues {FloatData [<-42 52 -70>, <-39 -125.8568 -70>]}}
       )
       Spawn Group Mask=2
      )
      Spawn Group Mask=1
     )
    )
    Sampled Position Offset Coordinate Space:ENiagaraSkelMeshTransforms=::NewEnumerator1
    Skeletal Mesh=$System.Skeletal Mesh
   )
   &ScaleMeshSize(Scale Factor=&MakeVector(
     X=1
     Y=1
     Z=&MaskFloatBySpawnGroup(
      Masked Float=&Lerp_Float(A=0.7, B=1.5, Alpha=$User.ParaGlideLeanAlpha)
      Passthrough Float=&MaskFloatBySpawnGroup(
       Masked Float=&Lerp_Float(A=1, B=1.8, Alpha=$User.ParaGlideLeanAlpha)
       Passthrough Float=&MaskFloatBySpawnGroup(
        Masked Float=&Lerp_Float(A=0.7, B=1.5, Alpha=$User.ParaGlideLeanAlpha)
        Passthrough Float=&Lerp_Float(A=1, B=1.8, Alpha=$User.ParaGlideLeanAlpha)
        Spawn Group Mask=2
       )
       Spawn Group Mask=1
      )
     )
    )) {module=ScaleMeshSize001}
   &DynamicMaterialParameters(
    Index 0 Param 1=&Lerp_Float(A=0.2, B=1, Alpha=$User.ParaGlideLeanAlpha)
    Index 0 Param 2=&Lerp_Float(A=0.2, B=1, Alpha=$User.ParaGlideLeanAlpha)
    Index 0 Param 3=&Lerp_Float(A=4, B=12, Alpha=$User.ParaGlideLeanAlpha)
    Module.Index0Param0Write=true
    Module.Index0Param1Write=true
    Module.Index0Param2Write=true
    Module.Index0Param3Write=true
   )
   /SetVariables(Particles.Scale=&Multiply_Vector(
     A=$Particles.Scale
     B=&MakeVector(
      X=1
      Y=1
      Z=&Lerp_Float(
       A=&FloatFromCurve(
        FloatCurve=DataInterface() {propertyValues {Curve {keys [
            {arriveTangent=12.3823, interpMode=RCIM_Cubic, leaveTangent=12.3823, tangentMode=RCTM_User, time=0, value=0}
            {arriveTangent=-1.125, interpMode=RCIM_Cubic, leaveTangent=-1.125, tangentMode=RCTM_User, time=0.2, value=1.5}
            {arriveTangent=-0.0405, interpMode=RCIM_Cubic, leaveTangent=-0.0405, tangentMode=RCTM_User, time=1.5, value=1}
           ]}}}
        CurveIndex=$Particles.Age
        Scale Curve=0.8
       )
       B=1
       Alpha=&FloatFromCurve(
        FloatCurve=DataInterface() {propertyValues {Curve {keys [{tangentMode=RCTM_User, time=1.5, value=0}, {tangentMode=RCTM_User, time=2, value=1}]}}}
        CurveIndex=$Particles.Age
       )
      )
     )
    )) {isSetParameters=true}
   &SkeletalMeshLocation(
    Bone / Socket Index=3
    Bone Sampling Mode=::NewEnumerator4
    MaskSpawnGroup=false
    Orientation Sampling=::NewEnumerator1
    Position Sampling=::NewEnumerator2
    Skeletal Mesh=$System.Skeletal Mesh
    Use Sampled Mesh Attribute as Ribbon ID=false
   ) {module=SkeletalMeshLocation003}
   /SetVariables(Particles.GlideLeanQuat:Quat4f=&LerpQuaternion(
     Quaternion A:Quat4f=&MultiplyQuaternion(
      Quaternion A=$Particles.SkeletalMeshLocation003.SampledRotation
      Quaternion B=&MakeQuatFromAxisAngle(Axis=<0 0 1>, AngleInDegrees=-15)
     )
     Quaternion B:Quat4f=&MultiplyQuaternion(
      Quaternion A=&MultiplyQuaternion(
       Quaternion A=&MultiplyQuaternion(
        Quaternion A=$Particles.SkeletalMeshLocation003.SampledRotation
        Quaternion B=&MakeQuatFromAxisAngle(Axis=<0 0 1>, AngleInDegrees=-15)
       )
       Quaternion B=&MultiplyQuaternion(
        Quaternion A=&MakeQuatFromAxisAngle(Axis=<1 0 0>, AngleInDegrees=&Lerp_Float(A=0, B=-50, Alpha=$System.Timeline.Time))
        Quaternion B=&MakeQuatFromAxisAngle(Axis=<1 0 0>, AngleInDegrees=&Lerp_Float(A=0, B=50, Alpha=$System.Timeline001.Time))
       )
      )
      Quaternion B=&MakeQuatFromAxisAngle(
       Axis=<0 0 1>
       AngleInDegrees=&Lerp_Float(
        A=0
        B=-60
        Alpha=&RemapRange(Input Value To Remap=$System.Dive, Input Min=0.28, Input Max=1, Output Min=0, Output Max=1)
       )
      )
     )
     Lerp Factor:NiagaraFloat=&FloatFromCurve(
      FloatCurve=DataInterface() {propertyValues {Curve {keys [
          {arriveTangent=2.5236, interpMode=RCIM_Cubic, leaveTangent=2.5236, tangentMode=RCTM_User, time=0, value=0}
          {interpMode=RCIM_Cubic, tangentMode=RCTM_User, time=0.9, value=1}
         ]}}}
      CurveIndex=$Emitter.Age
     )
     Enable Spherical Linear Interpolation:NiagaraBool=false
    )) {isSetParameters=true}
   &UpdateMeshOrientation(
    Facing Direction=&MakeVector(
     X=1
     Y=&RemapRange(
      Input Value To Remap=&MakeFloatFromVector(
       Vector=&RotateVectorByQuaternion(Vector To Rotate=<1 1 1>, Quaternion=$Particles.GlideLeanQuat)
       Channel=::NewEnumerator0
      )
      Input Min=0
      Input Max=-1.76
      Output Min=&Lerp_Float(A=-0.45, B=-0.75, Alpha=$System.Dive)
      Output Max=&Lerp_Float(A=0.45, B=0.75, Alpha=$System.Dive)
     )
     Z=&RemapRange(
      Input Value To Remap=&MakeFloatFromVector(
       Vector=&RotateVectorByQuaternion(Vector To Rotate=<1 1 1>, Quaternion=$Particles.GlideLeanQuat)
       Channel=::NewEnumerator0
      )
      Input Min=0
      Input Max=-1.76
      Output Min=&Lerp_Float(A=-0.45, B=0, Alpha=$System.Dive)
      Output Max=&Lerp_Float(A=0.45, B=0, Alpha=$System.Dive)
     )
    )
    Facing Interpolation=false
    Orientation Method=::NewEnumerator1
    Override Quat=false
    TransformFacing=false
    TransformUp=false
   )
   /SetVariables(Particles.Position=&AddVectorToPosition(
     Position=$Particles.Position
     Vector=&MakeVector(
      X=0
      Y=0
      Z=&Multiply_Float(
       A=1
       B=&Lerp_Float(
        A=500
        B=0
        Alpha=&FloatFromCurve(
         FloatCurve=DataInterface() {propertyValues {Curve {keys [
             {arriveTangent=4.1785, interpMode=RCIM_Cubic, leaveTangent=4.1785, tangentMode=RCTM_User, time=0, value=0}
             {arriveTangent=2.6161, interpMode=RCIM_Cubic, leaveTangent=2.6161, tangentMode=RCTM_User, time=0.2042, value=0.6228}
             {arriveTangent=0.0224, interpMode=RCIM_Cubic, leaveTangent=0.0224, tangentMode=RCTM_User, time=1, value=0.95}
             <at 1.5 1 RCIM_Cubic>
            ]}}}
         CurveIndex=$Emitter.Age
        )
       )
      )
     )
    )) {isSetParameters=true}
   /SetVariables(Particles.Position=&AddVectorToPosition(Position=$Particles.Position, Vector=$System.Ambient1)) {isSetParameters=true, module=SetVariables_#9}
   /SetVariables(Particles.Position=&AddVectorToPosition(
     Position=$Particles.Position
     Vector=&Lerp_Vector(
      A=<0 0 0>
      B=&FN_FrontEndVector3Modifier(
       Gameplay=&SelectVectorFromArray(
        Direct Array Index=$Particles.PartitionParticles.Partition
        Vector Selection Array=DataInterface() {propertyValues {FloatData [<0 -120 -85>, <0 15 55>]}}
       )
       Front End=&SelectVectorFromArray(
        Direct Array Index=$Particles.PartitionParticles.Partition
        Vector Selection Array=DataInterface() {propertyValues {FloatData [<15 25 -120>, <-5 30 5>]}}
       )
      )
      Alpha=$System.Dive
     )
    )) {isSetParameters=true}
   /SetVariables(
    Particles.SideToSideSlide:Vector3f=&TransformVector(
     Vector=&MakeVector(
      X=&MakeFloatFromVector(
       Vector=&VectorFromFloat(Value=&RemapRange(
         Input Value To Remap=&MakeFloatFromVector(
          Vector=&RotateVectorByQuaternion(Vector To Rotate=<1 1 1>, Quaternion=$Particles.GlideLeanQuat)
          Channel=::NewEnumerator0
         )
         Input Min=0
         Input Max=-1.76
         Output Min=&Lerp_Float(A=0, B=50, Alpha=$System.Dive)
         Output Max=&Lerp_Float(A=0, B=-50, Alpha=$System.Dive)
        ))
       Channel=::NewEnumerator0
      )
      Y=0
      Z=&Lerp_Float(A=0, B=50, Alpha=$System.Dive)
     )
     Source Space=::World
     Destination Space=::World
    )
    Particles.SidetoSideSlideBack:Vector3f=&TransformVector(
     Vector=&MakeVector(
      X=&MakeFloatFromVector(
       Vector=&VectorFromFloat(Value=&RemapRange(
         Input Value To Remap=&MakeFloatFromVector(
          Vector=&RotateVectorByQuaternion(Vector To Rotate=<1 1 1>, Quaternion=$Particles.GlideLeanQuat)
          Channel=::NewEnumerator0
         )
         Input Min=0
         Input Max=-1.76
         Output Min=&Lerp_Float(A=0, B=-120, Alpha=$System.Dive)
         Output Max=&Lerp_Float(A=0, B=120, Alpha=$System.Dive)
        ))
       Channel=::NewEnumerator0
      )
      Y=0
      Z=0
     )
     Source Space=::World
     Destination Space=::World
    )
   ) {isSetParameters=true}
   /SetVariables(Particles.Position=&MaskVector(
     Mask Source=1
     Mask Parameter=$Particles.Initial.PartitionParticles.Partition
     Mask=0
     Masked Vector=&AddVectorToPosition(Position=$Particles.Position, Vector=$Particles.SideToSideSlide)
     Passthrough Vector=&AddVectorToPosition(Position=$Particles.Position, Vector=$Particles.SidetoSideSlideBack)
    )) {isSetParameters=true}
  ]
  rendererClasses [&NiagaraMeshRendererProperties]
  renderers [
   {rendererClass=&NiagaraMeshRendererProperties, rendererIndex=0}
   {rendererClass=&NiagaraMeshRendererProperties, rendererIndex=1}
   {rendererClass=&NiagaraMeshRendererProperties, rendererIndex=2}
   {rendererClass=&NiagaraMeshRendererProperties, rendererIndex=3}
  ]
 }
 Tire_Back_02 {
  emitterUpdateScript [
   &EmitterState(
    Inactive Response=::NewEnumerator1
    Life Cycle Mode=::NewEnumerator1
    Loop Behavior=::NewEnumerator1
    Loop Duration Mode=::NewEnumerator1
    UseLoopDelay=false
   )
   &SpawnBurst_Instantaneous(Use Spawn Probability=false, UseLoopCountLimit=false)
   &SpawnBurst_Instantaneous(Spawn Group=1, Use Spawn Probability=false, UseLoopCountLimit=false) {module=SpawnBurst_Instantaneous001}
   &SpawnBurst_Instantaneous(Spawn Group=2, Use Spawn Probability=false, UseLoopCountLimit=false) {module=SpawnBurst_Instantaneous002}
   &SpawnBurst_Instantaneous(Spawn Group=3, Spawn Time=0.9, Use Spawn Probability=false, UseLoopCountLimit=false) {module=SpawnBurst_Instantaneous003}
   &SpawnBurst_Instantaneous(Spawn Group=5, Spawn Time=0.9, Use Spawn Probability=false, UseLoopCountLimit=false) {module=SpawnBurst_Instantaneous005}
  ]
  particleSpawnScript [
   &InitializeParticle(
    Mesh Scale=&MaskVectorBySpawnGroup(
     Masked Vector=<0.5 0.5 0.5>
     Passthrough Vector=&MaskVectorBySpawnGroup(
      Masked Vector=<0.5 0.5 0.5>
      Passthrough Vector=&MaskVectorBySpawnGroup(
       Masked Vector=<0.03 0.032 0.0006>
       Passthrough Vector=&MaskVectorBySpawnGroup(
        Masked Vector=<0.6 0.5 0.8>
        Passthrough Vector=&MaskVectorBySpawnGroup(Masked Vector=<0.05 0.35 0.7>, Passthrough Vector=<0.8 0.8 1>, Spawn Group Mask=4)
        Spawn Group Mask=3
       )
       Spawn Group Mask=2
      )
      Spawn Group Mask=1
     )
    )
    Mesh Scale Mode=::NewEnumerator4
    Sprite Rotation Mode=::NewEnumerator2
    UsePositionOffset=false
    Write Material Random=true
   )
   /SetVariables(Particles.TireFXColor:LinearColor=<rgba 5 0 0 5>) {isSetParameters=true, module=SetVariables_#10}
   /SetVariables(Particles.BlurFXColor:LinearColor=<rgba 0.24 0.24 0.24 7>) {isSetParameters=true, module=SetVariables_#11}
   /SetVariables(Particles.ShortTrail_Front_01_Color:LinearColor=<rgba 5 0.173 0.03 6>) {isSetParameters=true, module=SetVariables_#12}
   /SetVariables(Particles.Small_ShortTrail_Front_Color:LinearColor=<rgba 10 0 0 5>) {isSetParameters=true, module=SetVariables_#13}
   &InitialMeshOrientation(
    Mesh Orientation Mode=::NewEnumerator7
    Rotation=&MaskVectorBySpawnGroup(
     Masked Vector=<0 0 0>
     Passthrough Vector=&MaskVectorBySpawnGroup(
      Masked Vector=<0 0 0>
      Passthrough Vector=&MaskVectorBySpawnGroup(
       Masked Vector=<0 0.25 0>
       Passthrough Vector=&MaskVectorBySpawnGroup(
        Masked Vector=<0.031 0 0>
        Passthrough Vector=&MaskVectorBySpawnGroup(
         Masked Vector=<0.031 -0.02 0>
         Passthrough Vector=&MaskVectorBySpawnGroup(Masked Vector=<0.025 0 0>, Passthrough Vector=<0.045 0 0>, Spawn Group Mask=5)
         Spawn Group Mask=4
        )
        Spawn Group Mask=3
       )
       Spawn Group Mask=2
      )
      Spawn Group Mask=1
     )
    )
    Use Rotation Vector=true
   )
   &PartitionParticles()
  ]
  particleUpdateScript [
   &ParticleState(Kill Particles When Lifetime Has Elapsed=false, Let Infinitely Lived Particles Die When Emitter Deactivates=true)
   /SetVariables(Particles.VisibilityTag=&SpawnGroup()) {isSetParameters=true, module=SetVariables_#8}
   &ScaleMeshSize(Scale Factor=&MaskVectorBySpawnGroup(
     Masked Vector=<1 1 1>
     Passthrough Vector=&Lerp_Vector(
      A=<1 1 1>
      B=&MaskVectorBySpawnGroup(
       Masked Vector=<1.08 1 1.1>
       Passthrough Vector=&MaskVectorBySpawnGroup(
        Masked Vector=<1 1 1>
        Passthrough Vector=&MaskVectorBySpawnGroup(
         Masked Vector=<1 1 1.1>
         Passthrough Vector=&MaskVectorBySpawnGroup(Masked Vector=<1 1 1.1>, Passthrough Vector=<1 1 1>, Spawn Group Mask=4)
         Spawn Group Mask=3
        )
        Spawn Group Mask=2
       )
       Spawn Group Mask=1
      )
      Alpha=$System.Dive
     )
    ))
   &ScaleSpriteSize(
    Uniform Curve Index=$Emitter.Age
    Uniform Curve Scale=30
    Uniform Curve Sprite Scale=DataInterface() {propertyValues {Curve {keys [
        {arriveTangent=9.0725, interpMode=RCIM_Cubic, leaveTangent=9.0725, tangentMode=RCTM_User, time=1, value=0}
        {arriveTangent=-0.6609, interpMode=RCIM_Cubic, leaveTangent=-0.6609, tangentMode=RCTM_User, time=1.1, value=1}
        {arriveTangent=-0.0405, interpMode=RCIM_Cubic, leaveTangent=-0.0405, tangentMode=RCTM_User, time=1.4, value=0.5}
        <at 1.5 0 RCIM_Cubic>
       ]}}}
   )
   /SetVariables(Particles.TireSize:NiagaraFloat=&FloatFromCurve(FloatCurve=DataInterface() {propertyValues {Curve {keys [
         {arriveTangent=11.2563, interpMode=RCIM_Cubic, leaveTangent=11.2563, tangentMode=RCTM_User, time=0, value=0}
         {arriveTangent=0.0224, interpMode=RCIM_Cubic, leaveTangent=0.0224, tangentMode=RCTM_User, time=0.2, value=1}
        ]}}})) {isSetParameters=true, module=SetVariables_#14}
   /SetVariables(
    Particles.WheelEnergyBScale:Vector3f=&Multiply_Vector(A=<-0.22 0.15 0.15>, B=&VectorFromFloat(Value=$Particles.TireSize))
    Particles.WheelEnergyScale:Vector3f=&Multiply_Vector(A=<0.22 0.15 0.15>, B=&VectorFromFloat(Value=$Particles.TireSize))
   ) {isSetParameters=true, module=SetVariables_#15}
   &SkeletalMeshLocation(
    Bone / Socket Index=&SelectIntFromArray(Direct Array Index=$Particles.PartitionParticles.Partition, Int Selection Array=DataInterface() {propertyValues {IntData []}})
    Bone Sampling Mode=::NewEnumerator4
    MaskSpawnGroup=false
    Orientation Sampling=::NewEnumerator1
    Sampled Position Offset=&MaskVectorBySpawnGroup(
     Masked Vector=<-47.187 -130 -50>
     Passthrough Vector=&MaskVectorBySpawnGroup(
      Masked Vector=<-46.9464 -130 -49>
      Passthrough Vector=&MaskVectorBySpawnGroup(
       Masked Vector=<-46.5513 -130 -50>
       Passthrough Vector=&MaskVectorBySpawnGroup(
        Masked Vector=<-47.0854 -129.9588 -45.0336>
        Passthrough Vector=&MaskVectorBySpawnGroup(
         Masked Vector=&SelectVectorFromArray(
          Direct Array Index=$Particles.PartitionParticles.Partition
          Vector Selection Array=DataInterface() {propertyValues {FloatData [<-56 -131 -20>, <-56 -145 -20>]}}
         )
         Passthrough Vector=&MaskVectorBySpawnGroup(
          Masked Vector=<-47 -130 -45.0336>
          Passthrough Vector=&SelectVectorFromArray(
           Direct Array Index=$Particles.PartitionParticles.Partition
           Vector Selection Array=DataInterface() {propertyValues {FloatData [<-166071.8836 -142.4622 0>, <-3.8986 -144.4622 0>]}}
          )
          Spawn Group Mask=5
         )
         Spawn Group Mask=4
        )
        Spawn Group Mask=3
       )
       Spawn Group Mask=2
      )
      Spawn Group Mask=1
     )
    )
    Sampled Position Offset Coordinate Space:ENiagaraSkelMeshTransforms=::NewEnumerator2
    Skeletal Mesh=$System.Skeletal Mesh
   )
   &UpdateMeshOrientation(
    Rotation Rate=&MaskFloatBySpawnGroup(
     Masked Float=1
     Passthrough Float=&MaskFloatBySpawnGroup(
      Masked Float=0
      Passthrough Float=&MaskFloatBySpawnGroup(
       Masked Float=0
       Passthrough Float=&MaskFloatBySpawnGroup(
        Masked Float=0
        Passthrough Float=&MaskFloatBySpawnGroup(Masked Float=0, Passthrough Float=0, Spawn Group Mask=4)
        Spawn Group Mask=3
       )
       Spawn Group Mask=2
      )
      Spawn Group Mask=1
     )
    )
    Rotation Vector=<-1 0 0>
   ) {module=UpdateMeshOrientation001}
   &SkeletalMeshLocation(
    Bone / Socket Index=3
    Bone Sampling Mode=::NewEnumerator4
    MaskSpawnGroup=false
    Orientation Sampling=::NewEnumerator1
    Position Sampling=::NewEnumerator2
    Skeletal Mesh=$System.Skeletal Mesh
    Use Sampled Mesh Attribute as Ribbon ID=false
   ) {module=SkeletalMeshLocation001}
   /SetVariables(Particles.GlideLeanQuat:Quat4f=&LerpQuaternion(
     Quaternion A:Quat4f=&MultiplyQuaternion(
      Quaternion A=$Particles.SkeletalMeshLocation001.SampledRotation
      Quaternion B=&MakeQuatFromAxisAngle(Axis=<0 0 1>, AngleInDegrees=-15)
     )
     Quaternion B:Quat4f=&MultiplyQuaternion(
      Quaternion A=&MultiplyQuaternion(
       Quaternion A=&MultiplyQuaternion(
        Quaternion A=$Particles.SkeletalMeshLocation001.SampledRotation
        Quaternion B=&MakeQuatFromAxisAngle(Axis=<0 0 1>, AngleInDegrees=-15)
       )
       Quaternion B=&MultiplyQuaternion(
        Quaternion A=&MakeQuatFromAxisAngle(Axis=<1 0 0>, AngleInDegrees=&Lerp_Float(A=0, B=-50, Alpha=$System.Timeline.Time))
        Quaternion B=&MakeQuatFromAxisAngle(Axis=<1 0 0>, AngleInDegrees=&Lerp_Float(A=0, B=50, Alpha=$System.Timeline001.Time))
       )
      )
      Quaternion B=&MakeQuatFromAxisAngle(
       Axis=<0 0 1>
       AngleInDegrees=&Lerp_Float(
        A=0
        B=-60
        Alpha=&RemapRange(Input Value To Remap=$System.Dive, Input Min=0.28, Input Max=1, Output Min=0, Output Max=1)
       )
      )
     )
     Lerp Factor:NiagaraFloat=&FloatFromCurve(
      FloatCurve=DataInterface() {propertyValues {Curve {keys [
          {arriveTangent=2.5236, interpMode=RCIM_Cubic, leaveTangent=2.5236, tangentMode=RCTM_User, time=0, value=0}
          {interpMode=RCIM_Cubic, tangentMode=RCTM_User, time=0.9, value=1}
         ]}}}
      CurveIndex=$Emitter.Age
     )
     Enable Spherical Linear Interpolation:NiagaraBool=false
    )) {isSetParameters=true}
   &UpdateMeshOrientation(
    Facing Direction=&MakeVector(
     X=1
     Y=&RemapRange(
      Input Value To Remap=&MakeFloatFromVector(
       Vector=&RotateVectorByQuaternion(Vector To Rotate=<1 1 1>, Quaternion=$Particles.GlideLeanQuat)
       Channel=::NewEnumerator0
      )
      Input Min=0
      Input Max=-1.76
      Output Min=&Lerp_Float(A=-0.35, B=-0.75, Alpha=$System.Dive)
      Output Max=&Lerp_Float(A=0.35, B=0.75, Alpha=$System.Dive)
     )
     Z=&RemapRange(
      Input Value To Remap=&MakeFloatFromVector(
       Vector=&RotateVectorByQuaternion(Vector To Rotate=<1 1 1>, Quaternion=$Particles.GlideLeanQuat)
       Channel=::NewEnumerator0
      )
      Input Min=0
      Input Max=-1.76
      Output Min=&Lerp_Float(A=-0.35, B=0, Alpha=$System.Dive)
      Output Max=&Lerp_Float(A=0.35, B=0, Alpha=$System.Dive)
     )
    )
    Facing Interpolation=false
    Orientation Method=::NewEnumerator1
    Override Quat=false
    TransformFacing=false
    TransformUp=false
   )
   /SetVariables(Particles.Position=&AddVectorToPosition(
     Position=$Particles.Position
     Vector=&MakeVector(
      X=0
      Y=0
      Z=&Multiply_Float(
       A=1
       B=&Lerp_Float(
        A=500
        B=0
        Alpha=&FloatFromCurve(
         FloatCurve=DataInterface() {propertyValues {Curve {keys [
             {arriveTangent=4.1785, interpMode=RCIM_Cubic, leaveTangent=4.1785, tangentMode=RCTM_User, time=0, value=0}
             {arriveTangent=2.6161, interpMode=RCIM_Cubic, leaveTangent=2.6161, tangentMode=RCTM_User, time=0.2042, value=0.6228}
             {arriveTangent=0.0224, interpMode=RCIM_Cubic, leaveTangent=0.0224, tangentMode=RCTM_User, time=1, value=0.95}
             <at 1.5 1 RCIM_Cubic>
            ]}}}
         CurveIndex=$Emitter.Age
        )
       )
      )
     )
    )) {isSetParameters=true, module=SetVariables_#16}
   /SetVariables(Particles.Position=&AddVectorToPosition(Position=$Particles.Position, Vector=$System.Ambient1)) {isSetParameters=true, module=SetVariables_#9}
   /SetVariables(Particles.Position=&AddVectorToPosition(
     Position=$Particles.Position
     Vector=&Lerp_Vector(
      A=<0 0 0>
      B=&FN_FrontEndVector3Modifier(
       Gameplay=&MaskVectorBySpawnGroup(
        Masked Vector=<0 15 65>
        Passthrough Vector=&MaskVectorBySpawnGroup(
         Masked Vector=<-0.0145 15.4532 66.2167>
         Passthrough Vector=&MaskVectorBySpawnGroup(
          Masked Vector=<0 15 65>
          Passthrough Vector=&MaskVectorBySpawnGroup(
           Masked Vector=<2.1605 14 62.167>
           Passthrough Vector=&MaskVectorBySpawnGroup(
            Masked Vector=<-0.5 14 72.8928>
            Passthrough Vector=&MaskVectorBySpawnGroup(Masked Vector=<1.5995 17.3967 59.3732>, Passthrough Vector=<1.4781 17.3967 61.8384>, Spawn Group Mask=5)
            Spawn Group Mask=4
           )
           Spawn Group Mask=3
          )
          Spawn Group Mask=2
         )
         Spawn Group Mask=1
        )
       )
       Front End=&MaskVectorBySpawnGroup(
        Masked Vector=<-5 30 5>
        Passthrough Vector=&MaskVectorBySpawnGroup(
         Masked Vector=<-5 30 5>
         Passthrough Vector=&MaskVectorBySpawnGroup(
          Masked Vector=<-5 30 5>
          Passthrough Vector=&MaskVectorBySpawnGroup(
           Masked Vector=<-5 30 5>
           Passthrough Vector=&MaskVectorBySpawnGroup(
            Masked Vector=<-5 30 5>
            Passthrough Vector=&MaskVectorBySpawnGroup(Masked Vector=<-5 30 5>, Passthrough Vector=<-5 30 5>, Spawn Group Mask=5)
            Spawn Group Mask=4
           )
           Spawn Group Mask=3
          )
          Spawn Group Mask=2
         )
         Spawn Group Mask=1
        )
       )
      )
      Alpha=$System.Dive
     )
    )) {isSetParameters=true, module=SetVariables_#17}
   /SetVariables(Particles.SideToSideSlide:Vector3f=&TransformVector(
     Vector=&MakeVector(
      X=&RemapRange(
       Input Value To Remap=&MakeFloatFromVector(
        Vector=&RotateVectorByQuaternion(Vector To Rotate=<1 1 1>, Quaternion=$Particles.GlideLeanQuat)
        Channel=::NewEnumerator0
       )
       Input Min=0
       Input Max=-1.76
       Output Min=&Lerp_Float(A=0, B=-120, Alpha=$System.Dive)
       Output Max=&Lerp_Float(A=0, B=120, Alpha=$System.Dive)
      )
      Y=0
      Z=0
     )
     Source Space=::World
     Destination Space=::World
    )) {isSetParameters=true}
   /SetVariables(Particles.Position=&AddVectorToPosition(Position=$Particles.Position, Vector=$Particles.SideToSideSlide)) {isSetParameters=true}
   /SetVariables(
    Particles.FlareColor=&MakeLinearColor(
     R:NiagaraFloat=1
     G:NiagaraFloat=0
     B:NiagaraFloat=0
     A:NiagaraFloat=&FloatFromCurve(FloatCurve=DataInterface() {propertyValues {Curve {keys [<at 1 1>,<at 1.2 0>]}}}, CurveIndex=$Emitter.Age)
    )
    Particles.WheelEnergyColor:LinearColor=&Multiply_LinearColorByFloat(
     Linear Color=$Particles.WheelEnergyColor
     Float=&FloatFromCurve(
      FloatCurve=DataInterface() {propertyValues {Curve {keys [
          {interpMode=RCIM_Cubic, tangentMode=RCTM_User, time=1.3, value=1}
          {arriveTangent=1.3728, interpMode=RCIM_Cubic, leaveTangent=1.3728, tangentMode=RCTM_User, time=1.5, value=1.5}
          {arriveTangent=-0.54, interpMode=RCIM_Cubic, leaveTangent=-0.54, tangentMode=RCTM_User, time=2.1989, value=1}
         ]}}}
      CurveIndex=$Particles.Age
      Scale Curve=&FloatFromCurve(
       FloatCurve=DataInterface() {propertyValues {Curve {keys [{tangentMode=RCTM_User, time=1.2, value=5}, {tangentMode=RCTM_User, time=1.5, value=1}]}}}
       CurveIndex=$Particles.Age
      )
     )
    )
   ) {isSetParameters=true, module=SetVariables_#18}
   /SetVariables(Particles.TrailDynamicMat:Vector4f=<0.3 0.2 2 0>) {isSetParameters=true, module=SetVariables_#19}
   /SetVariables(Particles.ShortTrail_Front_L_Scale:Vector3f=&VectorFromCurve(
     VectorCurve=DataInterface() {propertyValues {
       XCurve {keys [<at 0 1>,<at 1 1>]}
       YCurve {keys [<at 0 1>,<at 1 1>]}
       ZCurve {keys [
         <at 0 0.4828 RCIM_Cubic>
         {arriveTangent=2.9353, interpMode=RCIM_Cubic, leaveTangent=2.9353, time=0.07, value=1.3446}
         <at 0.1762 1 RCIM_Cubic>
        ]}
      }}
     Scale Curve=<0.6 0.5 0.8>
    )) {isSetParameters=true}
  ]
  rendererClasses [&NiagaraMeshRendererProperties, &NiagaraSpriteRendererProperties]
  renderers [
   {rendererClass=&NiagaraMeshRendererProperties, rendererIndex=0}
   {rendererClass=&NiagaraMeshRendererProperties, rendererIndex=1}
   {rendererClass=&NiagaraMeshRendererProperties, rendererIndex=2}
   {rendererClass=&NiagaraSpriteRendererProperties, rendererIndex=3}
   {rendererClass=&NiagaraMeshRendererProperties, rendererIndex=4}
   {rendererClass=&NiagaraMeshRendererProperties, rendererIndex=5}
   {rendererClass=&NiagaraMeshRendererProperties, rendererIndex=6}
   {rendererClass=&NiagaraMeshRendererProperties, rendererIndex=7}
  ]
 }
 Tire_Front {
  emitterUpdateScript [
   &EmitterState(
    Inactive Response=::NewEnumerator1
    Life Cycle Mode=::NewEnumerator1
    Loop Behavior=::NewEnumerator1
    Loop Duration Mode=::NewEnumerator1
    UseLoopDelay=false
   )
   &SpawnBurst_Instantaneous(Use Spawn Probability=false, UseLoopCountLimit=false)
   &SpawnBurst_Instantaneous(Spawn Group=1, Use Spawn Probability=false, UseLoopCountLimit=false) {module=SpawnBurst_Instantaneous001}
   &SpawnBurst_Instantaneous(Spawn Group=2, Use Spawn Probability=false, UseLoopCountLimit=false) {module=SpawnBurst_Instantaneous002}
   &SpawnBurst_Instantaneous(Spawn Group=3, Spawn Time=0.9, Use Spawn Probability=false, UseLoopCountLimit=false) {module=SpawnBurst_Instantaneous003}
   &SpawnBurst_Instantaneous(Spawn Group=5, Spawn Time=0.9, Use Spawn Probability=false, UseLoopCountLimit=false) {module=SpawnBurst_Instantaneous005}
  ]
  particleSpawnScript [
   &InitializeParticle(
    Mesh Scale=&MaskVectorBySpawnGroup(
     Masked Vector=<0.5 0.5 0.5>
     Passthrough Vector=&MaskVectorBySpawnGroup(
      Masked Vector=<0.5 0.5 0.5>
      Passthrough Vector=&MaskVectorBySpawnGroup(
       Masked Vector=<0.03 0.032 0.0006>
       Passthrough Vector=&MaskVectorBySpawnGroup(
        Masked Vector=<0.5 0.5 0.8>
        Passthrough Vector=&MaskVectorBySpawnGroup(Masked Vector=<0.05 0.35 0.7>, Passthrough Vector=<0.8 1 1>, Spawn Group Mask=4)
        Spawn Group Mask=3
       )
       Spawn Group Mask=2
      )
      Spawn Group Mask=1
     )
    )
    Mesh Scale Mode=::NewEnumerator4
    Sprite Rotation Mode=::NewEnumerator2
    UsePositionOffset=false
    Write Material Random=true
   )
   /SetVariables(Particles.TireFXColor:LinearColor=<rgba 5 0 0 5>) {isSetParameters=true, module=SetVariables_#10}
   /SetVariables(Particles.BlurFXColor:LinearColor=<rgba 0.24 0.24 0.24 7>) {isSetParameters=true, module=SetVariables_#11}
   /SetVariables(Particles.ShortTrail_Front_01_Color:LinearColor=<rgba 5 0.173 0.03 6>) {isSetParameters=true, module=SetVariables_#12}
   /SetVariables(Particles.Small_ShortTrail_Front_Color:LinearColor=<rgba 10 0 0 5>) {isSetParameters=true, module=SetVariables_#13}
   &InitialMeshOrientation(
    Mesh Orientation Mode=::NewEnumerator7
    Rotation=&MaskVectorBySpawnGroup(
     Masked Vector=<0 0 0>
     Passthrough Vector=&MaskVectorBySpawnGroup(
      Masked Vector=<0 0 0>
      Passthrough Vector=&MaskVectorBySpawnGroup(
       Masked Vector=<0 0.25 0>
       Passthrough Vector=&MaskVectorBySpawnGroup(
        Masked Vector=<0.031 0 0>
        Passthrough Vector=&MaskVectorBySpawnGroup(
         Masked Vector=<0.031 -0.02 0>
         Passthrough Vector=&MaskVectorBySpawnGroup(Masked Vector=<0.03 0 0>, Passthrough Vector=<0.045 0 0>, Spawn Group Mask=5)
         Spawn Group Mask=4
        )
        Spawn Group Mask=3
       )
       Spawn Group Mask=2
      )
      Spawn Group Mask=1
     )
    )
    Use Rotation Vector=true
   )
   &PartitionParticles()
  ]
  particleUpdateScript [
   &ParticleState(Kill Particles When Lifetime Has Elapsed=false, Let Infinitely Lived Particles Die When Emitter Deactivates=true)
   /SetVariables(Particles.VisibilityTag=&SpawnGroup()) {isSetParameters=true, module=SetVariables_#8}
   &ScaleMeshSize(Scale Factor=&MaskVectorBySpawnGroup(
     Masked Vector=<1 1 1>
     Passthrough Vector=&Lerp_Vector(
      A=<1 1 1>
      B=&MaskVectorBySpawnGroup(
       Masked Vector=<1.08 1 1.1>
       Passthrough Vector=&MaskVectorBySpawnGroup(
        Masked Vector=<1 1 1>
        Passthrough Vector=&MaskVectorBySpawnGroup(
         Masked Vector=<1 1 1.1>
         Passthrough Vector=&MaskVectorBySpawnGroup(Masked Vector=<1 1 1.1>, Passthrough Vector=<1 1 1>, Spawn Group Mask=4)
         Spawn Group Mask=3
        )
        Spawn Group Mask=2
       )
       Spawn Group Mask=1
      )
      Alpha=$System.Dive
     )
    ))
   &ScaleSpriteSize(
    Uniform Curve Index=$Emitter.Age
    Uniform Curve Scale=30
    Uniform Curve Sprite Scale=DataInterface() {propertyValues {Curve {keys [
        {arriveTangent=9.0725, interpMode=RCIM_Cubic, leaveTangent=9.0725, tangentMode=RCTM_User, time=1, value=0}
        {arriveTangent=-0.6609, interpMode=RCIM_Cubic, leaveTangent=-0.6609, tangentMode=RCTM_User, time=1.1, value=1}
        {arriveTangent=-0.0405, interpMode=RCIM_Cubic, leaveTangent=-0.0405, tangentMode=RCTM_User, time=1.4, value=0.5}
        <at 1.5 0 RCIM_Cubic>
       ]}}}
   )
   /SetVariables(Particles.TireSize:NiagaraFloat=&FloatFromCurve(FloatCurve=DataInterface() {propertyValues {Curve {keys [
         {arriveTangent=11.2563, interpMode=RCIM_Cubic, leaveTangent=11.2563, tangentMode=RCTM_User, time=0, value=0}
         {arriveTangent=0.0224, interpMode=RCIM_Cubic, leaveTangent=0.0224, tangentMode=RCTM_User, time=0.2, value=1}
        ]}}})) {isSetParameters=true, module=SetVariables_#14}
   /SetVariables(
    Particles.WheelEnergyBScale:Vector3f=&Multiply_Vector(A=<-0.22 0.15 0.15>, B=&VectorFromFloat(Value=$Particles.TireSize))
    Particles.WheelEnergyScale:Vector3f=&Multiply_Vector(A=<0.22 0.15 0.15>, B=&VectorFromFloat(Value=$Particles.TireSize))
   ) {isSetParameters=true, module=SetVariables_#15}
   &SkeletalMeshLocation(
    Bone / Socket Index=&SelectIntFromArray(Direct Array Index=$Particles.PartitionParticles.Partition, Int Selection Array=DataInterface() {propertyValues {IntData []}})
    Bone Sampling Mode=::NewEnumerator4
    MaskSpawnGroup=false
    Orientation Sampling=::NewEnumerator1
    Sampled Position Offset=&MaskVectorBySpawnGroup(
     Masked Vector=<-47.187 50 -50>
     Passthrough Vector=&MaskVectorBySpawnGroup(
      Masked Vector=<-47.187 50 -49>
      Passthrough Vector=&MaskVectorBySpawnGroup(
       Masked Vector=<-46.5417 50 -50>
       Passthrough Vector=&MaskVectorBySpawnGroup(
        Masked Vector=<-47.3008 47.179 -47.9038>
        Passthrough Vector=&MaskVectorBySpawnGroup(
         Masked Vector=&SelectVectorFromArray(
          Direct Array Index=$Particles.PartitionParticles.Partition
          Vector Selection Array=DataInterface() {propertyValues {FloatData [<-56 50 -20>, <-56 33 -20>]}}
         )
         Passthrough Vector=&MaskVectorBySpawnGroup(
          Masked Vector=<-47.187 52.5484 -50>
          Passthrough Vector=&SelectVectorFromArray(
           Direct Array Index=$Particles.PartitionParticles.Partition
           Vector Selection Array=DataInterface() {propertyValues {FloatData [<-57.7201 37.5 0>, <-57.7201 36 0>]}}
          )
          Spawn Group Mask=5
         )
         Spawn Group Mask=4
        )
        Spawn Group Mask=3
       )
       Spawn Group Mask=2
      )
      Spawn Group Mask=1
     )
    )
    Sampled Position Offset Coordinate Space:ENiagaraSkelMeshTransforms=::NewEnumerator2
    Skeletal Mesh=$System.Skeletal Mesh
   )
   &UpdateMeshOrientation(
    Rotation Rate=&MaskFloatBySpawnGroup(
     Masked Float=1
     Passthrough Float=&MaskFloatBySpawnGroup(
      Masked Float=0
      Passthrough Float=&MaskFloatBySpawnGroup(
       Masked Float=0
       Passthrough Float=&MaskFloatBySpawnGroup(
        Masked Float=0
        Passthrough Float=&MaskFloatBySpawnGroup(Masked Float=0, Passthrough Float=0, Spawn Group Mask=4)
        Spawn Group Mask=3
       )
       Spawn Group Mask=2
      )
      Spawn Group Mask=1
     )
    )
    Rotation Vector=<-1 0 0>
   ) {module=UpdateMeshOrientation001}
   &SkeletalMeshLocation(
    Bone / Socket Index=3
    Bone Sampling Mode=::NewEnumerator4
    MaskSpawnGroup=false
    Orientation Sampling=::NewEnumerator1
    Position Sampling=::NewEnumerator2
    Skeletal Mesh=$System.Skeletal Mesh
    Use Sampled Mesh Attribute as Ribbon ID=false
   ) {module=SkeletalMeshLocation001}
   /SetVariables(Particles.GlideLeanQuat:Quat4f=&LerpQuaternion(
     Quaternion A:Quat4f=&MultiplyQuaternion(
      Quaternion A=$Particles.SkeletalMeshLocation001.SampledRotation
      Quaternion B=&MakeQuatFromAxisAngle(Axis=<0 0 1>, AngleInDegrees=-15)
     )
     Quaternion B:Quat4f=&MultiplyQuaternion(
      Quaternion A=&MultiplyQuaternion(
       Quaternion A=&MultiplyQuaternion(
        Quaternion A=$Particles.SkeletalMeshLocation001.SampledRotation
        Quaternion B=&MakeQuatFromAxisAngle(Axis=<0 0 1>, AngleInDegrees=-15)
       )
       Quaternion B=&MultiplyQuaternion(
        Quaternion A=&MakeQuatFromAxisAngle(Axis=<1 0 0>, AngleInDegrees=&Lerp_Float(A=0, B=-50, Alpha=$System.Timeline.Time))
        Quaternion B=&MakeQuatFromAxisAngle(Axis=<1 0 0>, AngleInDegrees=&Lerp_Float(A=0, B=50, Alpha=$System.Timeline001.Time))
       )
      )
      Quaternion B=&MakeQuatFromAxisAngle(
       Axis=<0 0 1>
       AngleInDegrees=&Lerp_Float(
        A=0
        B=-60
        Alpha=&RemapRange(Input Value To Remap=$System.Dive, Input Min=0.28, Input Max=1, Output Min=0, Output Max=1)
       )
      )
     )
     Lerp Factor:NiagaraFloat=&FloatFromCurve(
      FloatCurve=DataInterface() {propertyValues {Curve {keys [
          {arriveTangent=2.5236, interpMode=RCIM_Cubic, leaveTangent=2.5236, tangentMode=RCTM_User, time=0, value=0}
          {interpMode=RCIM_Cubic, tangentMode=RCTM_User, time=0.9, value=1}
         ]}}}
      CurveIndex=$Emitter.Age
     )
     Enable Spherical Linear Interpolation:NiagaraBool=false
    )) {isSetParameters=true}
   &UpdateMeshOrientation(
    Facing Direction=&MakeVector(
     X=1
     Y=&RemapRange(
      Input Value To Remap=&MakeFloatFromVector(
       Vector=&RotateVectorByQuaternion(Vector To Rotate=<1 1 1>, Quaternion=$Particles.GlideLeanQuat)
       Channel=::NewEnumerator0
      )
      Input Min=0
      Input Max=-1.76
      Output Min=&Lerp_Float(A=-0.45, B=-0.75, Alpha=$System.Dive)
      Output Max=&Lerp_Float(A=0.45, B=0.75, Alpha=$System.Dive)
     )
     Z=&RemapRange(
      Input Value To Remap=&MakeFloatFromVector(
       Vector=&RotateVectorByQuaternion(Vector To Rotate=<1 1 1>, Quaternion=$Particles.GlideLeanQuat)
       Channel=::NewEnumerator0
      )
      Input Min=0
      Input Max=-1.76
      Output Min=&Lerp_Float(A=-0.45, B=0, Alpha=$System.Dive)
      Output Max=&Lerp_Float(A=0.45, B=0, Alpha=$System.Dive)
     )
    )
    Facing Interpolation=false
    Orientation Method=::NewEnumerator1
    Override Quat=false
    TransformFacing=false
    TransformUp=false
   )
   /SetVariables(Particles.Position=&AddVectorToPosition(
     Position=$Particles.Position
     Vector=&MakeVector(
      X=0
      Y=0
      Z=&Multiply_Float(
       A=1
       B=&Lerp_Float(
        A=500
        B=0
        Alpha=&FloatFromCurve(
         FloatCurve=DataInterface() {propertyValues {Curve {keys [
             {arriveTangent=4.1785, interpMode=RCIM_Cubic, leaveTangent=4.1785, tangentMode=RCTM_User, time=0, value=0}
             {arriveTangent=2.6161, interpMode=RCIM_Cubic, leaveTangent=2.6161, tangentMode=RCTM_User, time=0.2042, value=0.6228}
             {arriveTangent=0.0224, interpMode=RCIM_Cubic, leaveTangent=0.0224, tangentMode=RCTM_User, time=1, value=0.95}
             <at 1.5 1 RCIM_Cubic>
            ]}}}
         CurveIndex=$Emitter.Age
        )
       )
      )
     )
    )) {isSetParameters=true, module=SetVariables_#16}
   /SetVariables(Particles.Position=&AddVectorToPosition(Position=$Particles.Position, Vector=$System.Ambient1)) {isSetParameters=true, module=SetVariables_#9}
   /SetVariables(Particles.Position=&AddVectorToPosition(
     Position=$Particles.Position
     Vector=&Lerp_Vector(
      A=<0 0 0>
      B=&FN_FrontEndVector3Modifier(
       Gameplay=&MaskVectorBySpawnGroup(
        Masked Vector=<0 -120 -60>
        Passthrough Vector=&MaskVectorBySpawnGroup(
         Masked Vector=<0 -120 -58.7695>
         Passthrough Vector=&MaskVectorBySpawnGroup(
          Masked Vector=<0 -120 -60>
          Passthrough Vector=&MaskVectorBySpawnGroup(
           Masked Vector=<2.0421 -120 -62.7843>
           Passthrough Vector=&MaskVectorBySpawnGroup(
            Masked Vector=<0 -120 -62>
            Passthrough Vector=&MaskVectorBySpawnGroup(Masked Vector=<1.1307 -120 -75.6863>, Passthrough Vector=<0.49 -120 -35.2593>, Spawn Group Mask=5)
            Spawn Group Mask=4
           )
           Spawn Group Mask=3
          )
          Spawn Group Mask=2
         )
         Spawn Group Mask=1
        )
       )
       Front End=&MaskVectorBySpawnGroup(Masked Vector=<15 5 -60>, Passthrough Vector=<15 25 -60>, Spawn Group Mask=5)
      )
      Alpha=$System.Dive
     )
    )) {isSetParameters=true, module=SetVariables_#17}
   /SetVariables(Particles.SideToSideSlide:Vector3f=&TransformVector(
     Vector=&MakeVector(
      X=&RemapRange(
       Input Value To Remap=&MakeFloatFromVector(
        Vector=&RotateVectorByQuaternion(Vector To Rotate=<1 1 1>, Quaternion=$Particles.GlideLeanQuat)
        Channel=::NewEnumerator0
       )
       Input Min=0
       Input Max=-1.76
       Output Min=&Lerp_Float(A=0, B=50, Alpha=$System.Dive)
       Output Max=&Lerp_Float(A=0, B=-50, Alpha=$System.Dive)
      )
      Y=0
      Z=0
     )
     Source Space=::World
     Destination Space=::World
    )) {isSetParameters=true}
   /SetVariables(Particles.Position=&AddVectorToPosition(Position=$Particles.Position, Vector=$Particles.SideToSideSlide)) {isSetParameters=true}
   /SetVariables(
    Particles.FlareColor=&MakeLinearColor(
     R:NiagaraFloat=1
     G:NiagaraFloat=0
     B:NiagaraFloat=0
     A:NiagaraFloat=&FloatFromCurve(FloatCurve=DataInterface() {propertyValues {Curve {keys [<at 1 1>,<at 1.2 0>]}}}, CurveIndex=$Emitter.Age)
    )
    Particles.WheelEnergyColor:LinearColor=&Multiply_LinearColorByFloat(
     Linear Color=$Particles.WheelEnergyColor
     Float=&FloatFromCurve(
      FloatCurve=DataInterface() {propertyValues {Curve {keys [
          {interpMode=RCIM_Cubic, tangentMode=RCTM_User, time=1.3, value=1}
          {arriveTangent=1.3728, interpMode=RCIM_Cubic, leaveTangent=1.3728, tangentMode=RCTM_User, time=1.5, value=1.5}
          {arriveTangent=-0.54, interpMode=RCIM_Cubic, leaveTangent=-0.54, tangentMode=RCTM_User, time=2.1989, value=1}
         ]}}}
      CurveIndex=$Particles.Age
      Scale Curve=&FloatFromCurve(
       FloatCurve=DataInterface() {propertyValues {Curve {keys [{tangentMode=RCTM_User, time=1.2, value=5}, {tangentMode=RCTM_User, time=1.5, value=1}]}}}
       CurveIndex=$Particles.Age
      )
     )
    )
   ) {isSetParameters=true, module=SetVariables_#18}
   /SetVariables(Particles.TrailDynamicMat:Vector4f=<0.3 0.2 2 0>) {isSetParameters=true, module=SetVariables_#19}
   /SetVariables(Particles.ShortTrail_Front_L_Scale:Vector3f=&VectorFromCurve(
     VectorCurve=DataInterface() {propertyValues {
       XCurve {keys [<at 0 1>,<at 1 1>]}
       YCurve {keys [<at 0 1>,<at 1 1>]}
       ZCurve {keys [
         <at 0 0.4828 RCIM_Cubic>
         {arriveTangent=2.9353, interpMode=RCIM_Cubic, leaveTangent=2.9353, time=0.07, value=1.3446}
         <at 0.1762 1 RCIM_Cubic>
        ]}
      }}
     Scale Curve=<0.6 0.5 0.8>
    )) {isSetParameters=true}
  ]
  rendererClasses [&NiagaraMeshRendererProperties, &NiagaraSpriteRendererProperties]
  renderers [
   {rendererClass=&NiagaraMeshRendererProperties, rendererIndex=0}
   {rendererClass=&NiagaraMeshRendererProperties, rendererIndex=1}
   {rendererClass=&NiagaraMeshRendererProperties, rendererIndex=2}
   {rendererClass=&NiagaraSpriteRendererProperties, rendererIndex=3}
   {rendererClass=&NiagaraMeshRendererProperties, rendererIndex=4}
   {rendererClass=&NiagaraMeshRendererProperties, rendererIndex=5}
   {rendererClass=&NiagaraMeshRendererProperties, rendererIndex=6}
   {rendererClass=&NiagaraMeshRendererProperties, rendererIndex=7}
  ]
 }
}
summary {
 systemName=NS_Contrail_KartRocketOriented
 emitters [
  {emitterName=Long_Trail, rendererClasses [&NiagaraMeshRendererProperties]}
  {emitterName=Tire_Front, rendererClasses [&NiagaraMeshRendererProperties, &NiagaraSpriteRendererProperties]}
  {emitterName=Dive_Dust_Front, rendererClasses [&NiagaraSpriteRendererProperties]}
  {emitterName=Dive_Dust_Back, rendererClasses [&NiagaraSpriteRendererProperties]}
  {emitterName=Tire_Back_02, rendererClasses [&NiagaraMeshRendererProperties, &NiagaraSpriteRendererProperties]}
 ]
 userVariables [
  User.bIsFrontEnd {type=NiagaraBool}
  User.bIsFrontEndPreview {type=NiagaraBool}
  User.ParaGlideLeanAlpha {type=NiagaraFloat, defaultValue=0, description=""}
 ]
}
